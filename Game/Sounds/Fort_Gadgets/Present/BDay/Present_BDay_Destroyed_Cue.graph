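# Sound cue for the destruction of a birthday present gadget.
# A Random node picks between two destroy wave variations with equal weight.
# It plays through the small-object demolition attenuation in the impacts sound class, with a slightly boosted volume.
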SoundCue /Game/Sounds/Fort_Gadgets/Present/BDay/Present_BDay_Destroyed_Cue
root=Random_1
nodes [
 {id=Random_1, children [WavePlayer_3,WavePlayer_4], properties {weights [1,1]}}
 {id=WavePlayer_3, properties {soundWaveAssetPtr=&/Present_BDay_Destroy_02}}
 {id=WavePlayer_4, properties {soundWaveAssetPtr=&/Present_BDay_Destroy_01}}
]
properties {
 attenuationSettings=/Game/Sounds/Attenuation/Fort_Demolition_Destroy_Small_Sized_Object_Att
 maxDistance=4000
 soundClassObject=&Fort_Impacts
 volumeMultiplier=1.2
}
