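# Sound cue for the in-air loop of the birthday variant of the Present gadget.
# A single looping wave player plays one in-air loop sample directly into the output, with no randomization or mixing.
# Playback runs at reduced volume in the gadget items sound class, spatialized through the shared gadget-activate attenuation settings.
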SoundCue /Game/Sounds/Fort_Gadgets/Present/BDay/Present_BDay_InAirLoop_Cue
root=WavePlayer_0
nodes [{id=WavePlayer_0, properties {bLooping=true, soundWaveAssetPtr=&/Present_BDay_InAirLoop_01}}]
properties {
 attenuationSettings=&Fort_GadgetActivate_Att
 maxDistance=4000
 soundClassObject=&Fort_Gadgets_Items
 volumeMultiplier=0.4
}
