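# Sound cue for placing the birthday present gadget in Fortnite, named as a place/crash sound.
# A Random node picks between two equally weighted placement wave variations.
# It plays in the gadget items sound class at reduced volume, with overridden attenuation that adds distance low-pass filtering and wide stereo spread alongside a referenced explosion attenuation asset.
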SoundCue /Game/Sounds/Fort_Gadgets/Present/BDay/Present_BDay_PlaceCrash_Cue
root=Random_0
nodes [
 {id=Random_0, children [WavePlayer_1,WavePlayer_2], properties {weights [1,1]}}
 {id=WavePlayer_1, properties {soundWaveAssetPtr=&/Present_BDay_Place_02_final}}
 {id=WavePlayer_2, properties {soundWaveAssetPtr=&/Present_BDay_Place_01_final}}
]
properties {
 attenuationSettings=/Game/Sounds/Attenuation/Fort_Explosion_Medium_StW_Att
 bOverrideAttenuation=true
 maxDistance=7500
 soundClassObject=&Fort_Gadgets_Items
 volumeMultiplier=0.5
 attenuationOverrides {bAttenuateWithLPF=true, falloffDistance=5000, lPFFrequencyAtMax=50, stereoSpread=5000, attenuationShapeExtents {x=2500}}
}
