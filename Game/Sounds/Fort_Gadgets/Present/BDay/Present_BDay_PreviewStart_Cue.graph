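# Sound cue for the start of placement preview on the birthday present gadget, layering several sounds into one quiet, attenuated gadget-item effect.
# An inner mixer combines the present pickup sound with one of two randomly chosen aim sounds.
# An outer mixer adds a machine pistol reload foley wave through a modulator that fixes its pitch at double, with both layers boosted in the mix.
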SoundCue /Game/Sounds/Fort_Gadgets/Present/BDay/Present_BDay_PreviewStart_Cue
root=Mixer_0
nodes [
 {id=Mixer_0, children [Mixer_1,Modulator_0], properties {inputVolume [1.5,1.2]}}
 {id=Mixer_1, children [WavePlayer_4,Random_0], properties {inputVolume [1,1]}}
 {id=Modulator_0, children [WavePlayer_1], properties {pitchMax=2, pitchMin=2}}
 {id=Random_0, children [WavePlayer_2,WavePlayer_3], properties {weights [1,1]}}
 {id=WavePlayer_1, properties {soundWaveAssetPtr=/Game/Sounds/Fort_Weapons/pistol/pistol_foley/machine_pistol_reload_start}}
 {id=WavePlayer_2, properties {soundWaveAssetPtr=&/Present_BDay_Aim_02}}
 {id=WavePlayer_3, properties {soundWaveAssetPtr=&/Present_BDay_Aim_01}}
 {id=WavePlayer_4, properties {soundWaveAssetPtr=&/Present_BDay_Pickup_01}}
]
properties {
 attenuationSettings=&Fort_GadgetActivate_Att
 maxDistance=4000
 soundClassObject=&Fort_Gadgets_Items
 volumeMultiplier=0.2
}
