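# Sound cue for throwing the birthday (BDay) present gadget.
# A Random node picks between two throw wave variations with equal weight.
# Playback is at reduced volume, routed to the gadget items sound class and spatialized with the shared gadget-activate attenuation.
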SoundCue /Game/Sounds/Fort_Gadgets/Present/BDay/Present_BDay_Throw_Cue
root=Random_0
nodes [
 {id=Random_0, children [WavePlayer_1,WavePlayer_2], properties {weights [1,1]}}
 {id=WavePlayer_1, properties {soundWaveAssetPtr=&/Present_BDay_Throw_02}}
 {id=WavePlayer_2, properties {soundWaveAssetPtr=&/Present_BDay_Throw_01}}
]
properties {
 attenuationSettings=&Fort_GadgetActivate_Att
 maxDistance=4000
 soundClassObject=&Fort_Gadgets_Items
 volumeMultiplier=0.4
}
